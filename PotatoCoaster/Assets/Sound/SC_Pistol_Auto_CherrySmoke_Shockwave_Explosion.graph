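# Sound cue for the shockwave explosion of the CherrySmoke auto pistol, reusing Shockwave Launcher explosion waves in the weapons sound class.
# A mixer layers a close and a distant branch, each with its own attenuation asset, so the explosion's character changes with listener distance.
# Each branch feeds one of three randomly picked waves through a modulator.
# A concurrency override caps simultaneous plays per owner, stopping the oldest, and random-node branch culling is disabled.
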SoundCue /PotatoCoaster/Assets/Sound/SC_Pistol_Auto_CherrySmoke_Shockwave_Explosion
root=Mixer_2
nodes [
 {id=Attenuation_0, children [Modulator_1], properties {attenuationSettings=/Game/Sounds/Attenuation/Fort_Shockwave_Launcher_Explo_Close_Att}}
 {
  id=Attenuation_4
  children [Modulator_6]
  properties {attenuationSettings=/Game/Sounds/Attenuation/Fort_Shockwave_Launcher_Explo_Distant_Att}
 }
 {id=Mixer_2, children [Attenuation_0, Attenuation_4], properties {inputVolume [1,1]}}
 {id=Modulator_1, children [Random_1]}
 {id=Modulator_6, children [Random_2]}
 {id=Random_1, children [WavePlayer_0,WavePlayer_1,WavePlayer_2], properties {weights [1,1,1]}}
 {id=Random_2, children [WavePlayer_3,WavePlayer_4,WavePlayer_5], properties {weights [1,1,1]}}
 {id=WavePlayer_0, properties {soundWaveAssetPtr=&/Shockwave_Launcher_Explode_Close_01}}
 {id=WavePlayer_1, properties {soundWaveAssetPtr=&/Shockwave_Launcher_Explode_Close_02}}
 {id=WavePlayer_2, properties {soundWaveAssetPtr=&/Shockwave_Launcher_Explode_Close_03}}
 {id=WavePlayer_3, properties {soundWaveAssetPtr=&/Shockwave_Launcher_Explode_Distant_03}}
 {id=WavePlayer_4, properties {soundWaveAssetPtr=&/Shockwave_Launcher_Explode_Distant_02}}
 {id=WavePlayer_5, properties {soundWaveAssetPtr=&/Shockwave_Launcher_Explode_Distant_01}}
]
properties {
 bExcludeFromRandomNodeBranchCulling=true
 bOverrideConcurrency=true
 maxDistance=13000
 soundClassObject=/Game/Sounds/FortSoundClassesAndModes/Fort_Weapons
 volumeMultiplier=0.5
 concurrencyOverrides {bLimitToOwner=true, maxCount=3, resolutionRule=StopOldest, volumeScaleAttackTime=0.1}
}
