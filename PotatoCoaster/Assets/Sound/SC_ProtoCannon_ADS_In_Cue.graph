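# Sound cue for the ProtoCannon weapon's aim-down-sights entry, routed to the weapon foley sound class.
# A mixer layers a random pick of four player movement foley waves, a briefly delayed main ADS wave and an undelayed sweetener, each with its own fixed volume.
# Overall volume is scaled well down, attenuation is overridden on the cue, and concurrency allows one instance, stopping the oldest.
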SoundCue /PotatoCoaster/Assets/Sound/SC_ProtoCannon_ADS_In_Cue
root=Mixer_0
nodes [
 {id=Delay_0, children [Modulator_1], properties {delayMax=0.05, delayMin=0.05}}
 {id=Mixer_0, children [Modulator_0,Delay_0,Modulator_2], properties {inputVolume [1,1.5,1]}}
 {id=Modulator_0, children [Random_4], properties {volumeMax=0.8, volumeMin=0.8}}
 {id=Modulator_1, children [WavePlayer_10], properties {pitchMax=1, pitchMin=1, volumeMax=3, volumeMin=3}}
 {id=Modulator_2, children [WavePlayer_5], properties {pitchMax=1, pitchMin=1, volumeMax=2.5, volumeMin=2.5}}
 {id=Random_4, children [WavePlayer_0,WavePlayer_1,WavePlayer_2,WavePlayer_3], properties {weights [1,1,1,1]}}
 {id=WavePlayer_0, properties {soundWaveAssetPtr=/Game/Sounds/Fort_Foley_Sounds/Moves/player_run_foley_15}}
 {id=WavePlayer_1, properties {soundWaveAssetPtr=/Game/Sounds/Fort_Foley_Sounds/Moves/player_run_foley_14}}
 {id=WavePlayer_10, properties {soundWaveAssetPtr=&/S40_ProtoCannon_ADS_In_01}}
 {id=WavePlayer_2, properties {soundWaveAssetPtr=/Game/Sounds/Fort_Foley_Sounds/Moves/player_run_foley_13}}
 {id=WavePlayer_3, properties {soundWaveAssetPtr=/Game/Sounds/Fort_Foley_Sounds/Moves/player_run_foley_12}}
 {id=WavePlayer_5, properties {soundWaveAssetPtr=&/S40_ProtoCannon_ADS_In_Sweetener}}
]
properties {
 attenuationSettings=&Fort_Weapon_Projectile_FirstPerson_Athena_Att
 bOverrideAttenuation=true
 bOverrideConcurrency=true
 maxDistance=2700
 soundClassObject=&Fort_WeaponFoley
 volumeMultiplier=0.1
 attenuationOverrides {falloffDistance=1500, nonSpatializedRadiusStart=200, attenuationShapeExtents {x=1200}}
 concurrencyOverrides {maxCount=1, resolutionRule=StopOldest}
}
