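# Sound cue for the close-range projectile whiz-by of the ProtoCannon weapon.
# It mixes a random ProtoCannon passby with a half-volume, pitched-down grenade-launcher whip layer, then applies an outer Modulator at default settings.
# A quality-level switch feeds the whip layer from a five-wave random pool in its first two slots and a three-wave subset in the last two.
# Attenuation is overridden on the cue, and a shared bullet whiz-by concurrency rule applies.
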SoundCue /PotatoCoaster/Assets/Sound/SC_ProtoCannon_BulletWhip_Close_Cue
root=Modulator_2
nodes [
 {id=Mixer_1, children [Random_4,Modulator_1], properties {inputVolume [1,0.5]}}
 {id=Modulator_1, children [QualityLevel_1], properties {pitchMax=0.7, pitchMin=0.7}}
 {id=Modulator_2, children [Mixer_1]}
 {id=QualityLevel_1, children [Random_2,Random_2,Random_3,Random_3]}
 {id=Random_2, children [WavePlayer_0,WavePlayer_1,WavePlayer_3,WavePlayer_4,WavePlayer_5], properties {weights [1,1,1,1,1]}}
 {id=Random_3, children [WavePlayer_3,WavePlayer_4,WavePlayer_5], properties {bShouldExcludeFromBranchCulling=true, weights [1,1,1]}}
 {id=Random_4, children [WavePlayer_10, WavePlayer_11, WavePlayer_12, WavePlayer_13], properties {weights [1,1,1,1]}}
 {
  id=WavePlayer_0
  properties {soundWaveAssetPtr=/HopscotchWeaponsGameplay/Items/GrenadeLauncher/Sound/StickyGrenadeLauncher_GrenadeWhip_Close_01}
 }
 {
  id=WavePlayer_1
  properties {soundWaveAssetPtr=/HopscotchWeaponsGameplay/Items/GrenadeLauncher/Sound/StickyGrenadeLauncher_GrenadeWhip_Close_02}
 }
 {id=WavePlayer_10, properties {soundWaveAssetPtr=&/S40_ProtoCannon_Passby_Close_01}}
 {id=WavePlayer_11, properties {soundWaveAssetPtr=&/S40_ProtoCannon_Passby_Close_02}}
 {id=WavePlayer_12, properties {soundWaveAssetPtr=&/S40_ProtoCannon_Passby_Close_03}}
 {id=WavePlayer_13, properties {soundWaveAssetPtr=&/S40_ProtoCannon_Passby_Close_04}}
 {
  id=WavePlayer_3
  properties {soundWaveAssetPtr=/HopscotchWeaponsGameplay/Items/GrenadeLauncher/Sound/StickyGrenadeLauncher_GrenadeWhip_Close_04}
 }
 {
  id=WavePlayer_4
  properties {soundWaveAssetPtr=/HopscotchWeaponsGameplay/Items/GrenadeLauncher/Sound/StickyGrenadeLauncher_GrenadeWhip_Close_05}
 }
 {
  id=WavePlayer_5
  properties {soundWaveAssetPtr=/HopscotchWeaponsGameplay/Items/GrenadeLauncher/Sound/StickyGrenadeLauncher_GrenadeWhip_Close_06}
 }
]
properties {
 attenuationSettings=/Game/Athena/Sounds/Weapons/BulletWhips/bullet_whip_Att.Bullet_Whip_Att
 bOverrideAttenuation=true
 maxDistance=4000
 priority=2
 soundClassObject=&Fort_WeaponRanged_Projectile_Effects
 volumeMultiplier=0.9
 attenuationOverrides {nonSpatializedRadiusStart=200, stereoSpread=500}
 concurrencySet [/Game/Sounds/Concurrency_Rules/Athena_BulletWhizBy_Conc]
}
