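# Sound cue for the far bullet whip of the ProtoCannon, the sound of a projectile passing by at distance.
# A random node picks one of four equally weighted passby waves and feeds a modulator that holds pitch steady.
# The cue lowers overall pitch and volume, plays in the ranged weapon projectile effects class, and references shared bullet whip attenuation and whiz-by concurrency assets.
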SoundCue /PotatoCoaster/Assets/Sound/SC_ProtoCannon_BulletWhip_Far_Cue
root=Modulator_2
nodes [
 {id=Modulator_2, children [Random_4], properties {pitchMax=1, pitchMin=1}}
 {id=Random_4, children [WavePlayer_10, WavePlayer_11, WavePlayer_12, WavePlayer_13], properties {weights [1,1,1,1]}}
 {id=WavePlayer_10, properties {soundWaveAssetPtr=&/S40_ProtoCannon_Passby_Far_01}}
 {id=WavePlayer_11, properties {soundWaveAssetPtr=&/S40_ProtoCannon_Passby_Far_02}}
 {id=WavePlayer_12, properties {soundWaveAssetPtr=&/S40_ProtoCannon_Passby_Far_03}}
 {id=WavePlayer_13, properties {soundWaveAssetPtr=&/S40_ProtoCannon_Passby_Far_04}}
]
properties {
 attenuationSettings=/Game/Athena/Sounds/Weapons/BulletWhips/bullet_whip_Att.Bullet_Whip_Att
 maxDistance=2097152
 pitchMultiplier=0.6
 soundClassObject=&Fort_WeaponRanged_Projectile_Effects
 volumeMultiplier=0.5
 attenuationOverrides {bAttenuate=false, stereoSpread=0}
 concurrencySet [/Game/Sounds/Concurrency_Rules/Athena_BulletWhizBy_Conc]
}
