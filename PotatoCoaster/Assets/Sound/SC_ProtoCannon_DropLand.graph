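# Sound cue for the ProtoCannon weapon landing when dropped, judging by its name.
# A random node picks evenly between two reused StickyGrenadeLauncher DropHand waves, passed through a modulator fixed at neutral pitch and volume.
# The cue raises pitch and lowers volume overall, using the Fort weapon-drop attenuation, the loot-collect sound class and the Weapon_Drop_StopOldest concurrency set.
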SoundCue /PotatoCoaster/Assets/Sound/SC_ProtoCannon_DropLand
root=Modulator_0
nodes [
 {id=Modulator_0, children [Random_0], properties {pitchMax=1, pitchMin=1, volumeMax=1, volumeMin=1}}
 {id=Random_0, children [WavePlayer_0,WavePlayer_1], properties {weights [1,1]}}
 {id=WavePlayer_0, properties {soundWaveAssetPtr=/HopscotchWeaponsGameplay/Items/GrenadeLauncher/Sound/StickyGrenadeLauncher_DropHand_01}}
 {id=WavePlayer_1, properties {soundWaveAssetPtr=/HopscotchWeaponsGameplay/Items/GrenadeLauncher/Sound/StickyGrenadeLauncher_DropHand_02}}
]
properties {
 attenuationSettings=&Fort_Weapon_Drop_Att
 maxDistance=3000
 pitchMultiplier=1.5
 soundClassObject=&Fort_Loot_Collect
 volumeMultiplier=0.3
 concurrencyOverrides {resolutionRule=PreventNew}
 concurrencySet [&Weapon_Drop_StopOldest]
}
