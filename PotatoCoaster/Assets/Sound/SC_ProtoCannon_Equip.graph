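# Sound cue for equipping the ProtoCannon, classed and attenuated as a weapon pickup and loot collect sound.
# A mixer layers the main equip wave, given slight pitch variation by a modulator, with a much quieter sweetener wave that starts after a short fixed delay.
# The cue references the PickupMaximum_Oldest concurrency set to limit overlapping pickup sounds.
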SoundCue /PotatoCoaster/Assets/Sound/SC_ProtoCannon_Equip
root=Mixer_0
nodes [
 {id=Delay_0, children [WavePlayer_1], properties {delayMax=0.1, delayMin=0.1}}
 {id=Mixer_0, children [Modulator_0,Delay_0], properties {inputVolume [1,0.2]}}
 {id=Modulator_0, children [WavePlayer_3], properties {pitchMin=0.99, volumeMax=1, volumeMin=1}}
 {id=WavePlayer_1, properties {soundWaveAssetPtr=&/S40_ProtoCannon_Equip_Sweetener}}
 {id=WavePlayer_3, properties {soundWaveAssetPtr=&/S40_ProtoCannon_Equip}}
]
properties {
 attenuationSettings=&Fort_Weapon_Pickup_Att
 maxDistance=3500
 soundClassObject=&Fort_Loot_Collect
 volumeMultiplier=0.7
 concurrencyOverrides {maxCount=1, resolutionRule=PreventNew}
 concurrencySet [&PickupMaximum_Oldest]
}
