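# Sound cue for picking up the ProtoCannon, layering borrowed grenade launcher pickup sounds with a cannon reload tail.
# A mixer plays one of two random sticky grenade launcher pickup waves at once and the ProtoCannon reload-end wave after a short fixed delay.
# Each layer passes through its own modulator, and the whole cue is slightly lowered in pitch and volume.
# It uses weapon pickup attenuation, the loot-collect sound class and a shared pickup concurrency set.
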SoundCue /PotatoCoaster/Assets/Sound/SC_ProtoCannon_Pickup
root=Mixer_0
nodes [
 {id=Delay_0, children [Modulator_1], properties {delayMax=0.2, delayMin=0.2}}
 {id=Mixer_0, children [Modulator_0,Delay_0], properties {inputVolume [1,1]}}
 {id=Modulator_0, children [Random_0]}
 {id=Modulator_1, children [WavePlayer_3]}
 {id=Random_0, children [WavePlayer_1,WavePlayer_2], properties {weights [1,1]}}
 {id=WavePlayer_1, properties {soundWaveAssetPtr=/HopscotchWeaponsGameplay/Items/GrenadeLauncher/Sound/StickyGrenadeLauncher_PickUp_02}}
 {id=WavePlayer_2, properties {soundWaveAssetPtr=/HopscotchWeaponsGameplay/Items/GrenadeLauncher/Sound/StickyGrenadeLauncher_PickUp_03}}
 {id=WavePlayer_3, properties {soundWaveAssetPtr=&/S40_ProtoCannon_Reload_End_03}}
]
properties {
 attenuationSettings=&Fort_Weapon_Pickup_Att
 maxDistance=3500
 pitchMultiplier=0.9
 soundClassObject=&Fort_Loot_Collect
 volumeMultiplier=0.7
 concurrencyOverrides {maxCount=1, resolutionRule=PreventNew}
 concurrencySet [&PickupMaximum_Oldest]
}
