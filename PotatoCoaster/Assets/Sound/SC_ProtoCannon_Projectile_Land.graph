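# Sound cue for a ProtoCannon projectile landing, played as a medium explosive with dedicated attenuation.
# A mixer layers three randomized branches: a reused ClusterBomb throw-land sound, a ClusterBomb warning beep and a ProtoCannon impact explosion.
# The beep branch starts after a short fixed delay with raised pitch, while the landing layer is boosted slightly in the mix.
# Each branch picks among equally weighted variations and passes through a modulator for variation.
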SoundCue /PotatoCoaster/Assets/Sound/SC_ProtoCannon_Projectile_Land
root=Mixer_0
nodes [
 {id=Delay_0, children [Modulator_1], properties {delayMax=0.1, delayMin=0.1}}
 {id=Mixer_0, children [Modulator_0,Delay_0,Modulator_2], properties {inputVolume [1.1,1,1]}}
 {id=Modulator_0, children [Random_0]}
 {id=Modulator_1, children [Random_1], properties {pitchMax=1.4, pitchMin=1.2, volumeMin=0.9}}
 {id=Modulator_2, children [Random_2], properties {pitchMax=1, pitchMin=1, volumeMin=0.9}}
 {
  id=Random_0
  children [WavePlayer_0,WavePlayer_1,WavePlayer_2,WavePlayer_3]
  properties {bShouldExcludeFromBranchCulling=true, weights [1,1,1,1]}
 }
 {id=Random_1, children [WavePlayer_4,WavePlayer_5,WavePlayer_6], properties {bShouldExcludeFromBranchCulling=true, weights [1,1,1]}}
 {id=Random_2, children [WavePlayer_7,WavePlayer_8,WavePlayer_9], properties {bShouldExcludeFromBranchCulling=true, weights [1,1,1]}}
 {id=WavePlayer_0, properties {soundWaveAssetPtr=/ClusterBombGameplay/Audio/ClusterBomb_ThrowLand_01}}
 {id=WavePlayer_1, properties {soundWaveAssetPtr=/ClusterBombGameplay/Audio/ClusterBomb_ThrowLand_02}}
 {id=WavePlayer_2, properties {soundWaveAssetPtr=/ClusterBombGameplay/Audio/ClusterBomb_ThrowLand_03}}
 {id=WavePlayer_3, properties {soundWaveAssetPtr=/ClusterBombGameplay/Audio/ClusterBomb_ThrowLand_04}}
 {id=WavePlayer_4, properties {soundWaveAssetPtr=/ClusterBombGameplay/Audio/ClusterBomb_WarningBeep_03}}
 {id=WavePlayer_5, properties {soundWaveAssetPtr=/ClusterBombGameplay/Audio/ClusterBomb_WarningBeep_02}}
 {id=WavePlayer_6, properties {soundWaveAssetPtr=/ClusterBombGameplay/Audio/ClusterBomb_WarningBeep_01}}
 {id=WavePlayer_7, properties {soundWaveAssetPtr=&/S40_ProtoCannon_ImpactExplode_01}}
 {id=WavePlayer_8, properties {soundWaveAssetPtr=&/S40_ProtoCannon_ImpactExplode_02}}
 {id=WavePlayer_9, properties {soundWaveAssetPtr=&/S40_ProtoCannon_ImpactExplode_03}}
]
properties {
 attenuationSettings=/PotatoCoaster/Assets/Sound/Settings/ProtoCannon_ClusterBomb_Explode
 maxDistance=18500
 soundClassObject=/Game/Sounds/FortSoundClassesAndModes/Fort_Explosives_Medium
 volumeMultiplier=0.8
}
