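# Sound cue for the end of the ProtoCannon weapon's reload.
# It randomly picks one of three equally weighted reload-end variants and mixes it with a quieter sweetener layer.
# A modulator adds slight pitch variation, and the cue applies a raised pitch and lowered volume.
# It uses Fortnite weapon-reload attenuation and the medium explosives sound class.
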SoundCue /PotatoCoaster/Assets/Sound/SC_ProtoCannon_Reload_End
root=Modulator_2
nodes [
 {id=Mixer_1, children [Random_0,WavePlayer_4], properties {inputVolume [1,0.7]}}
 {id=Modulator_2, children [Mixer_1], properties {pitchMax=1.01, pitchMin=0.97, volumeMax=1, volumeMin=1}}
 {id=Random_0, children [WavePlayer_0,WavePlayer_1,WavePlayer_2], properties {weights [1,1,1]}}
 {id=WavePlayer_0, properties {soundWaveAssetPtr=&/S40_ProtoCannon_Reload_End_03}}
 {id=WavePlayer_1, properties {soundWaveAssetPtr=&/S40_ProtoCannon_Reload_End_02}}
 {id=WavePlayer_2, properties {soundWaveAssetPtr=&/S40_ProtoCannon_Reload_End_01}}
 {id=WavePlayer_4, properties {soundWaveAssetPtr=&/S40_ProtoCannon_ReloadEnd_Sweetner}}
]
properties {
 attenuationSettings=&Fort_Weapon_Reload_Att
 maxDistance=2250
 pitchMultiplier=1.1
 soundClassObject=/Game/Sounds/FortSoundClassesAndModes/Fort_Explosives_Medium
 volumeMultiplier=0.6
}
