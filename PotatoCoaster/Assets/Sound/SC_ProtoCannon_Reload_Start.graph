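# Sound cue for the start of the ProtoCannon weapon reload.
# Mixes two randomized layers of three waves each: the main reload sound, slightly delayed with subtle pitch variation, and a quieter sweetener that plays immediately at fixed pitch.
# Output is scaled down in volume and uses Fortnite weapon reload attenuation with the medium explosives sound class.
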SoundCue /PotatoCoaster/Assets/Sound/SC_ProtoCannon_Reload_Start
root=Mixer_1
nodes [
 {id=Delay_1, children [Modulator_2], properties {delayMax=0.05, delayMin=0.05}}
 {id=Mixer_1, children [Delay_1,Modulator_1], properties {inputVolume [1,0.8]}}
 {id=Modulator_1, children [Random_1], properties {pitchMax=1, pitchMin=1, volumeMax=1, volumeMin=1}}
 {id=Modulator_2, children [Random_0], properties {pitchMax=1.01, pitchMin=0.97, volumeMax=1, volumeMin=1}}
 {id=Random_0, children [WavePlayer_2,WavePlayer_3,WavePlayer_4], properties {weights [1,1,1]}}
 {id=Random_1, children [WavePlayer_9, WavePlayer_10, WavePlayer_11], properties {weights [1,1,1]}}
 {id=WavePlayer_10, properties {soundWaveAssetPtr=&/S40_ProtoCannon_ReloadStart_Sweetner_02}}
 {id=WavePlayer_11, properties {soundWaveAssetPtr=&/S40_ProtoCannon_ReloadStart_Sweetner_03}}
 {id=WavePlayer_2, properties {soundWaveAssetPtr=&/S40_ProtoCannon_Reload_Start_01}}
 {id=WavePlayer_3, properties {soundWaveAssetPtr=&/S40_ProtoCannon_Reload_Start_02}}
 {id=WavePlayer_4, properties {soundWaveAssetPtr=&/S40_ProtoCannon_Reload_Start_03}}
 {id=WavePlayer_9, properties {soundWaveAssetPtr=&/S40_ProtoCannon_ReloadStart_Sweetner_01}}
]
properties {
 attenuationSettings=&Fort_Weapon_Reload_Att
 maxDistance=2250
 soundClassObject=/Game/Sounds/FortSoundClassesAndModes/Fort_Explosives_Medium
 volumeMultiplier=0.3
}
